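# Physics asset for the TennisLeash off-hand pickaxe, setting up its dangling bells, rings, bows and fur for physical secondary motion.
# Each dyn_ bone has one capsule body, with dyn_base_joint as the common root of the constraints.
# The fur is angularly locked to the base, while two bell-to-ring chains and four separate bows hang off it with modest swing limits and near-zero twist limits.
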PhysicsAsset /BRCosmetics/Weapons/FORT_Melee/Pickaxe_TennisLeash_OffHand/Meshes/TennisLeash_OffHand_Axe_Physics
bodies {
 dyn_base_joint [{
   shapeType=Capsule
   length=1.0447
   radius=0.6735
   shapeName=dyn_base_joint_capsule
   center=<-0.8888 1.5803 -0.9106>
   rotation=<rot -1.6167 -90.6279 1.7426>
  }]
 dyn_bell_1_joint [{
   shapeType=Capsule
   length=2.6859
   radius=2.6661
   shapeName=dyn_bell_1_joint_capsule
   center=<4.6584 -0.0978 0.028>
   rotation=<rot 0 -88.2668 -89.479>
  }]
 dyn_bell_2_joint [{
   shapeType=Capsule
   length=3.5547
   radius=2.0456
   shapeName=dyn_bell_2_joint_capsule
   center=<4.7286 -0.0836 0.0374>
   rotation=<rot 0 -88.3568 -89.241>
  }]
 dyn_bow_1_joint [{
   shapeType=Capsule
   length=1.7903
   radius=2.0534
   shapeName=dyn_bow_1_joint_capsule
   center=<3.1878 -0.3357 0.1143>
   rotation=<rot 8.7999 -90.0001 90.1063>
  }]
 dyn_bow_2_joint [{
   shapeType=Capsule
   length=1.834
   radius=2.047
   shapeName=dyn_bow_2_joint_capsule
   center=<3.3367 -0.1162 -0.0612>
   rotation=<rot -4.0064 -89.9998 -88.4244>
  }]
 dyn_bow_3_joint [{
   shapeType=Capsule
   length=4.032
   radius=1.4267
   shapeName=dyn_bow_3_joint_capsule
   center=<4.0731 -0.8101 -0.0997>
   rotation=<rot 0 90.5412 -88.2242>
  }]
 dyn_bow_4_joint [{
   shapeType=Capsule
   length=5.0399
   radius=1.2976
   shapeName=dyn_bow_4_joint_capsule
   center=<4.2522 -0.1701 0.0273>
   rotation=<rot 0 -88.7729 -89.6236>
  }]
 dyn_fur_joint [{
   shapeType=Capsule
   length=7.3644
   radius=7.4779
   shapeName=dyn_fur_joint_capsule
   center=<-0.9679 -1.4209 -0.9285>
   rotation=<rot 0.3323 89.8577 -89.8085>
  }]
 dyn_ring_1_joint [{
   shapeType=Capsule
   length=4.9586
   radius=0.8968
   shapeName=dyn_ring_1_joint_capsule
   center=<5.1771 0.0111 0.0092>
   rotation=<rot 0 -90.3224 -89.7154>
  }]
 dyn_ring_2_joint [{
   shapeType=Capsule
   length=4.8345
   radius=0.8709
   shapeName=dyn_ring_2_joint_capsule
   center=<4.9348 0.0138 0.0081>
   rotation=<rot 0 -90.3924 -89.7553>
  }]
}
constraints [
 {
  bone1Name=dyn_fur_joint
  bone2Name=dyn_base_joint
  swing1LimitDegrees=45
  swing1Motion=Locked
  swing2LimitDegrees=45
  swing2Motion=Locked
  twistLimitDegrees=45
  twistMotion=Locked
 }
 {bone1Name=dyn_bell_1_joint, bone2Name=dyn_base_joint, swing1LimitDegrees=12, swing2LimitDegrees=12, twistLimitDegrees=1}
 {bone1Name=dyn_ring_1_joint, bone2Name=dyn_bell_1_joint, swing1LimitDegrees=22, swing2LimitDegrees=22, twistLimitDegrees=1}
 {bone1Name=dyn_bell_2_joint, bone2Name=dyn_base_joint, swing1LimitDegrees=10, swing2LimitDegrees=10, twistLimitDegrees=1}
 {bone1Name=dyn_ring_2_joint, bone2Name=dyn_bell_2_joint, swing1LimitDegrees=22, swing2LimitDegrees=22, twistLimitDegrees=1}
 {bone1Name=dyn_bow_1_joint, bone2Name=dyn_base_joint, swing1LimitDegrees=10, swing2LimitDegrees=20, twistLimitDegrees=1}
 {bone1Name=dyn_bow_2_joint, bone2Name=dyn_base_joint, swing1LimitDegrees=10, swing2LimitDegrees=20, twistLimitDegrees=1}
 {bone1Name=dyn_bow_3_joint, bone2Name=dyn_base_joint, swing1LimitDegrees=10, swing2LimitDegrees=20, twistLimitDegrees=1}
 {bone1Name=dyn_bow_4_joint, bone2Name=dyn_base_joint, swing1LimitDegrees=10, swing2LimitDegrees=20, twistLimitDegrees=1}
]
